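# Physics asset for secondary motion of dangling details on the male medium Street Urchin body cosmetic: chest and hip cans, collar flaps and two zippers.
# Each dynamic bone has a capsule body constrained to its own tiny-capsule _MSTR counterpart, forming nine independent pairs rather than chains.
# Angular limits vary by part: chest cans swing and twist most freely, hip cans swing less, zippers barely twist, and collars are restricted in one swing axis and twist.
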PhysicsAsset /BRCosmetics/Characters/Player/Male/Medium/Bodies/M_MED_Street_Urchin/Meshes/M_MED_Street_Urchin_Dyn
bodies {
 dyn_can_chest_01 [{shapeType=Capsule, length=2, radius=2, center=<0.8588 0 0>, rotation=<rot 90 0 0>}]
 dyn_can_chest_01_MSTR [{shapeType=Capsule, length=0.2, radius=0.2, rotation=<rot 90 0 0>}]
 dyn_can_chest_02 [{shapeType=Capsule, length=2, radius=2, center=<0.8588 0 0>, rotation=<rot 90 0 0>}]
 dyn_can_chest_02_MSTR [{shapeType=Capsule, length=0.2, radius=0.2, rotation=<rot 90 0 0>}]
 dyn_can_chest_03 [{shapeType=Capsule, length=2, radius=2, center=<0.8588 0 0>, rotation=<rot 90 0 0>}]
 dyn_can_chest_03_MSTR [{shapeType=Capsule, length=0.2, radius=0.2, rotation=<rot 90 0 0>}]
 dyn_can_hip_01 [{shapeType=Capsule, length=2, radius=2, center=<0.8588 0 0>, rotation=<rot 90 0 0>}]
 dyn_can_hip_01_MSTR [{shapeType=Capsule, length=0.2, radius=0.2, rotation=<rot 90 0 0>}]
 dyn_can_hip_02 [{shapeType=Capsule, length=2, radius=2, center=<0.8588 0 0>, rotation=<rot 90 0 0>}]
 dyn_can_hip_02_MSTR [{shapeType=Capsule, length=0.2, radius=0.2, rotation=<rot 90 0 0>}]
 dyn_collar_l [{shapeType=Capsule, length=2, radius=2, center=<1 0 0>, rotation=<rot 90 0 0>}]
 dyn_collar_l_MSTR [{shapeType=Capsule, length=0.2, radius=0.2, rotation=<rot 90 0 0>}]
 dyn_collar_r [{shapeType=Capsule, length=2, radius=2, center=<1 0 0>, rotation=<rot 90 0 0>}]
 dyn_collar_r_MSTR [{shapeType=Capsule, length=0.2, radius=0.2, rotation=<rot 90 0 0>}]
 dyn_zipper_butt [{shapeType=Capsule, length=3.2303, radius=0.7697, center=<1.884 0 0>, rotation=<rot 90 0 0>}]
 dyn_zipper_butt_MSTR [{shapeType=Capsule, length=0.2, radius=0.2, rotation=<rot 90 0 0>}]
 dyn_zipper_chest [{shapeType=Capsule, length=3.2303, radius=0.7697, center=<1.884 0 0>, rotation=<rot 90 0 0>}]
 dyn_zipper_chest_MSTR [{shapeType=Capsule, length=0.2, radius=0.2, rotation=<rot 90 0 0>}]
}
constraints [
 {bone1Name=dyn_zipper_chest, bone2Name=dyn_zipper_chest_MSTR, swing1LimitDegrees=45, swing2LimitDegrees=45, twistLimitDegrees=5}
 {bone1Name=dyn_can_chest_01, bone2Name=dyn_can_chest_01_MSTR, swing1LimitDegrees=45, swing2LimitDegrees=45, twistLimitDegrees=45}
 {bone1Name=dyn_can_chest_02, bone2Name=dyn_can_chest_02_MSTR, swing1LimitDegrees=45, swing2LimitDegrees=45, twistLimitDegrees=45}
 {bone1Name=dyn_can_chest_03, bone2Name=dyn_can_chest_03_MSTR, swing1LimitDegrees=45, swing2LimitDegrees=45, twistLimitDegrees=45}
 {bone1Name=dyn_collar_l, bone2Name=dyn_collar_l_MSTR, swing1LimitDegrees=45, swing2LimitDegrees=10, twistLimitDegrees=5}
 {bone1Name=dyn_collar_r, bone2Name=dyn_collar_r_MSTR, swing1LimitDegrees=45, swing2LimitDegrees=10, twistLimitDegrees=5}
 {bone1Name=dyn_zipper_butt, bone2Name=dyn_zipper_butt_MSTR, swing1LimitDegrees=45, swing2LimitDegrees=45, twistLimitDegrees=5}
 {bone1Name=dyn_can_hip_02, bone2Name=dyn_can_hip_02_MSTR, swing1LimitDegrees=30, swing2LimitDegrees=30, twistLimitDegrees=45}
 {bone1Name=dyn_can_hip_01, bone2Name=dyn_can_hip_01_MSTR, swing1LimitDegrees=30, swing2LimitDegrees=30, twistLimitDegrees=45}
]
