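# Physics asset for the Street Urchin male medium body's left foot, setting up secondary motion for a dynamic shoe tongue.
# Capsule bodies cover the calf, foot and two tongue bones, chained by constraints from calf to foot to tongue.
# The two tongue joints have tight twist limits and asymmetric swing ranges, narrower on the tip bone, while the foot-to-calf joint leaves swing and twist free.
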PhysicsAsset /BRCosmetics/Characters/Player/Male/Medium/Bodies/M_MED_Street_Urchin/Meshes/M_MED_Street_Urchin_Dyn_foot_l
bodies {
 calf_l [{shapeType=Capsule, length=13.0098, radius=4.0811, center=<-33.0615 -1.1287 -0.0196>, rotation=<rot 90 178.2101 -176.7898>}]
 dyn_shoe_tongue_l_01 [{shapeType=Capsule, length=2, radius=2, center=<1 0 0>, rotation=<rot 90 0 0>}]
 dyn_shoe_tongue_l_02 [{shapeType=Capsule, length=5.6569, radius=0.892, center=<1 0 0>, rotation=<rot 90 0 0>}]
 foot_l [{shapeType=Capsule, length=0.5, radius=0.5, rotation=<rot 90 0 0>}]
}
constraints [
 {bone1Name=dyn_shoe_tongue_l_01, bone2Name=foot_l, swing1LimitDegrees=20, swing2LimitDegrees=90, twistLimitDegrees=5}
 {bone1Name=dyn_shoe_tongue_l_02, bone2Name=dyn_shoe_tongue_l_01, swing1LimitDegrees=10, swing2LimitDegrees=45, twistLimitDegrees=5}
 {
  bone1Name=foot_l
  bone2Name=calf_l
  swing1LimitDegrees=45
  swing1Motion=Free
  swing2LimitDegrees=45
  swing2Motion=Free
  twistLimitDegrees=45
  twistMotion=Free
 }
]
