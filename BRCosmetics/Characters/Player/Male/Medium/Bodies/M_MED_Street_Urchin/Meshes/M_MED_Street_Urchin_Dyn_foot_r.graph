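# Physics asset for the right foot of the Street Urchin medium male body, simulating a dynamic shoe tongue.
# Capsule bodies sit on the calf, foot and two shoe-tongue bones.
# Tongue bones chain from the foot with tight twist limits and wider second-axis swing, so the flap bends mainly one way.
# The foot-to-calf joint has free swing and twist.
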PhysicsAsset /BRCosmetics/Characters/Player/Male/Medium/Bodies/M_MED_Street_Urchin/Meshes/M_MED_Street_Urchin_Dyn_foot_r
bodies {
 calf_r [{shapeType=Capsule, length=12.7423, radius=4.3486, center=<33.0615 1.1287 0.0196>, rotation=<rot 90 178.2101 -176.7898>}]
 dyn_shoe_tongue_r_01 [{shapeType=Capsule, length=2, radius=2, center=<1 0 0>, rotation=<rot 90 0 0>}]
 dyn_shoe_tongue_r_02 [{shapeType=Capsule, length=6.4454, radius=0.8334, center=<1 0 0>, rotation=<rot 90 0 0>}]
 foot_r [{shapeType=Capsule, length=2, radius=2, rotation=<rot 90 0 0>}]
}
constraints [
 {bone1Name=dyn_shoe_tongue_r_01, bone2Name=foot_r, swing1LimitDegrees=20, swing2LimitDegrees=90, twistLimitDegrees=5}
 {bone1Name=dyn_shoe_tongue_r_02, bone2Name=dyn_shoe_tongue_r_01, swing1LimitDegrees=10, swing2LimitDegrees=45, twistLimitDegrees=5}
 {
  bone1Name=foot_r
  bone2Name=calf_r
  swing1LimitDegrees=45
  swing1Motion=Free
  swing2LimitDegrees=45
  swing2Motion=Free
  twistLimitDegrees=45
  twistMotion=Free
 }
]
