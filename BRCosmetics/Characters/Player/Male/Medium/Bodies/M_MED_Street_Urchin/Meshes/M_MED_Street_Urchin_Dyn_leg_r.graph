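# Physics asset for the right-leg dynamic bones of the male medium Street Urchin player body, covering the top and bottom pants flaps and thigh zipper.
# Each piece pairs a tiny capsule on its MSTR bone with a longer capsule offset along the dynamic bone.
# Three independent constraints join the pairs with limited swing and tight twist, and the flaps get a narrower first swing axis than the zipper.
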PhysicsAsset /BRCosmetics/Characters/Player/Male/Medium/Bodies/M_MED_Street_Urchin/Meshes/M_MED_Street_Urchin_Dyn_leg_r
bodies {
 dyn_pants_flap_r_bott [{shapeType=Capsule, length=3.2303, radius=0.7697, center=<1.884 0 0>, rotation=<rot 90 0 0>}]
 dyn_pants_flap_r_bott_MSTR [{shapeType=Capsule, length=0.2, radius=0.2, rotation=<rot 90 0 0>}]
 dyn_pants_flap_r_top [{shapeType=Capsule, length=3.2303, radius=0.7697, center=<1.884 0 0>, rotation=<rot 90 0 0>}]
 dyn_pants_flap_r_top_MSTR [{shapeType=Capsule, length=0.2, radius=0.2, rotation=<rot 90 0 0>}]
 dyn_zipper_thigh_r [{shapeType=Capsule, length=3.2303, radius=0.7697, center=<1.884 0 0>, rotation=<rot 90 0 0>}]
 dyn_zipper_thigh_r_MSTR [{shapeType=Capsule, length=0.2, radius=0.2, rotation=<rot 90 0 0>}]
}
constraints [
 {bone1Name=dyn_zipper_thigh_r, bone2Name=dyn_zipper_thigh_r_MSTR, swing1LimitDegrees=45, swing2LimitDegrees=45, twistLimitDegrees=5}
 {bone1Name=dyn_pants_flap_r_top, bone2Name=dyn_pants_flap_r_top_MSTR, swing1LimitDegrees=10, swing2LimitDegrees=45, twistLimitDegrees=5}
 {bone1Name=dyn_pants_flap_r_bott, bone2Name=dyn_pants_flap_r_bott_MSTR, swing1LimitDegrees=10, swing2LimitDegrees=45, twistLimitDegrees=5}
]
