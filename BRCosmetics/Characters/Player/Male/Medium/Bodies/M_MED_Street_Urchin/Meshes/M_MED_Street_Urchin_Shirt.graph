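# Physics asset for the male medium Street Urchin cosmetic shirt, defining bodies and joints for its dyn_shirt bones.
# Capsules cover pelvis, spine and thighs, with tiny capsules on each shirt chain's MSTR root.
# Left, center and right box chains hang from spine_04 for upper back, and spine_02 for lower back and front.
# Segment joints have tight swing and twist limits, while root and body joints leave rotation free.
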PhysicsAsset /BRCosmetics/Characters/Player/Male/Medium/Bodies/M_MED_Street_Urchin/Meshes/M_MED_Street_Urchin_Shirt
bodies {
 dyn_shirt_bk_bott_01 [{shapeType=Box, extentX=7.1855, extentY=12.9624, extentZ=0.8437, center=<3.2203 0 0>}]
 dyn_shirt_bk_bott_MSTR [{shapeType=Capsule, length=0.5, radius=0.5}]
 dyn_shirt_bk_bott_l_01 [{shapeType=Box, extentX=7.1855, extentY=12.9624, extentZ=0.8437, center=<3.2203 0 0>}]
 dyn_shirt_bk_bott_l_MSTR [{shapeType=Capsule, length=0.5, radius=0.5}]
 dyn_shirt_bk_bott_r_01 [{shapeType=Box, extentX=7.1855, extentY=12.9624, extentZ=0.8437, center=<3.2203 0 0>}]
 dyn_shirt_bk_bott_r_MSTR [{shapeType=Capsule, length=0.5, radius=0.5}]
 dyn_shirt_bk_top_01 [{shapeType=Box, extentX=7.1855, extentY=12.9624, extentZ=0.8437, center=<3.2203 0 0>}]
 dyn_shirt_bk_top_02 [{shapeType=Box, extentX=7.1855, extentY=12.9624, extentZ=0.8437, center=<3.2203 0 0>}]
 dyn_shirt_bk_top_MSTR [{shapeType=Capsule, length=0.5, radius=0.5}]
 dyn_shirt_bk_top_l_01 [{shapeType=Box, extentX=7.1855, extentY=12.9624, extentZ=1.8307, center=<3.2203 0 0>}]
 dyn_shirt_bk_top_l_02 [{shapeType=Box, extentX=7.1855, extentY=12.9624, extentZ=1.8307, center=<3.2203 0 0>}]
 dyn_shirt_bk_top_l_MSTR [{shapeType=Capsule, length=0.5, radius=0.5}]
 dyn_shirt_bk_top_r_01 [{shapeType=Box, extentX=7.1855, extentY=12.9624, extentZ=1.8307, center=<3.2203 0 0>}]
 dyn_shirt_bk_top_r_02 [{shapeType=Box, extentX=7.1855, extentY=12.9624, extentZ=1.8307, center=<3.2203 0 0>}]
 dyn_shirt_bk_top_r_MSTR [{shapeType=Capsule, length=0.5, radius=0.5}]
 dyn_shirt_fr_01 [{shapeType=Box, extentX=7.1855, extentY=12.9624, extentZ=0.8437, center=<3.2203 0 0>}]
 dyn_shirt_fr_02 [{shapeType=Box, extentX=7.1855, extentY=12.9624, extentZ=0.8437, center=<3.2203 0 0>}]
 dyn_shirt_fr_MSTR [{shapeType=Capsule, length=0.5, radius=0.5}]
 dyn_shirt_fr_l_01 [{shapeType=Box, extentX=7.1855, extentY=12.9624, extentZ=0.8437, center=<3.2203 0 0>}]
 dyn_shirt_fr_l_02 [{shapeType=Box, extentX=7.1855, extentY=12.9624, extentZ=0.8437, center=<3.2203 0 0>}]
 dyn_shirt_fr_l_MSTR [{shapeType=Capsule, length=0.5, radius=0.5}]
 dyn_shirt_fr_r_01 [{shapeType=Box, extentX=7.1855, extentY=12.9624, extentZ=0.8437, center=<3.2203 0 0>}]
 dyn_shirt_fr_r_02 [{shapeType=Box, extentX=7.1855, extentY=12.9624, extentZ=0.8437, center=<3.2203 0 0>}]
 dyn_shirt_fr_r_MSTR [{shapeType=Capsule, length=0.5, radius=0.5}]
 pelvis [{shapeType=Capsule, length=6.9129, radius=11.0088}]
 spine_02 [{shapeType=Capsule, length=3.9369, radius=11.0088, center=<0.0458 1.8912 0>}]
 spine_03 [{shapeType=Capsule, length=7.6459, radius=12.0285, center=<3.9572 0.882 0>}]
 spine_04 [{shapeType=Capsule, length=17.4584, radius=6.3987, center=<0.0458 1.2863 0>}]
 thigh_l [{shapeType=Capsule, length=27.3085, radius=10.1752, center=<-14.889 0 0.0001>, rotation=<rot 90 0 0>}]
 thigh_r [{shapeType=Capsule, length=27.3085, radius=10.1752, center=<14.889 0 -0.0001>, rotation=<rot 90 0 0>}]
}
constraints [
 {bone1Name=dyn_shirt_bk_top_l_01, bone2Name=dyn_shirt_bk_top_l_MSTR, swing1LimitDegrees=10, swing2LimitDegrees=25, twistLimitDegrees=2}
 {bone1Name=dyn_shirt_bk_top_l_02, bone2Name=dyn_shirt_bk_top_l_01, swing1LimitDegrees=10, swing2LimitDegrees=25, twistLimitDegrees=2}
 {bone1Name=dyn_shirt_bk_top_01, bone2Name=dyn_shirt_bk_top_MSTR, swing1LimitDegrees=10, swing2LimitDegrees=25, twistLimitDegrees=2}
 {bone1Name=dyn_shirt_bk_top_02, bone2Name=dyn_shirt_bk_top_01, swing1LimitDegrees=10, swing2LimitDegrees=25, twistLimitDegrees=2}
 {bone1Name=dyn_shirt_bk_top_r_01, bone2Name=dyn_shirt_bk_top_r_MSTR, swing1LimitDegrees=10, swing2LimitDegrees=25, twistLimitDegrees=2}
 {bone1Name=dyn_shirt_bk_top_r_02, bone2Name=dyn_shirt_bk_top_r_01, swing1LimitDegrees=10, swing2LimitDegrees=25, twistLimitDegrees=2}
 {bone1Name=dyn_shirt_bk_bott_l_01, bone2Name=dyn_shirt_bk_bott_l_MSTR, swing1LimitDegrees=10, swing2LimitDegrees=25, twistLimitDegrees=2}
 {bone1Name=dyn_shirt_bk_bott_01, bone2Name=dyn_shirt_bk_bott_MSTR, swing1LimitDegrees=10, swing2LimitDegrees=25, twistLimitDegrees=2}
 {bone1Name=dyn_shirt_bk_bott_r_01, bone2Name=dyn_shirt_bk_bott_r_MSTR, swing1LimitDegrees=10, swing2LimitDegrees=25, twistLimitDegrees=2}
 {bone1Name=dyn_shirt_fr_l_01, bone2Name=dyn_shirt_fr_l_MSTR, swing1LimitDegrees=10, swing2LimitDegrees=25, twistLimitDegrees=2}
 {bone1Name=dyn_shirt_fr_l_02, bone2Name=dyn_shirt_fr_l_01, swing1LimitDegrees=10, swing2LimitDegrees=25, twistLimitDegrees=2}
 {bone1Name=dyn_shirt_fr_01, bone2Name=dyn_shirt_fr_MSTR, swing1LimitDegrees=10, swing2LimitDegrees=25, twistLimitDegrees=2}
 {bone1Name=dyn_shirt_fr_02, bone2Name=dyn_shirt_fr_01, swing1LimitDegrees=10, swing2LimitDegrees=25, twistLimitDegrees=2}
 {bone1Name=dyn_shirt_fr_r_01, bone2Name=dyn_shirt_fr_r_MSTR, swing1LimitDegrees=10, swing2LimitDegrees=25, twistLimitDegrees=2}
 {bone1Name=dyn_shirt_fr_r_02, bone2Name=dyn_shirt_fr_r_01, swing1LimitDegrees=10, swing2LimitDegrees=25, twistLimitDegrees=2}
 {
  bone1Name=dyn_shirt_bk_top_l_MSTR
  bone2Name=spine_04
  swing1LimitDegrees=45
  swing1Motion=Free
  swing2LimitDegrees=45
  swing2Motion=Free
  twistLimitDegrees=45
  twistMotion=Free
 }
 {
  bone1Name=dyn_shirt_bk_top_MSTR
  bone2Name=spine_04
  swing1LimitDegrees=45
  swing1Motion=Free
  swing2LimitDegrees=45
  swing2Motion=Free
  twistLimitDegrees=45
  twistMotion=Free
 }
 {
  bone1Name=dyn_shirt_bk_top_r_MSTR
  bone2Name=spine_04
  swing1LimitDegrees=45
  swing1Motion=Free
  swing2LimitDegrees=45
  swing2Motion=Free
  twistLimitDegrees=45
  twistMotion=Free
 }
 {
  bone1Name=spine_03
  bone2Name=spine_02
  swing1LimitDegrees=45
  swing1Motion=Free
  swing2LimitDegrees=45
  swing2Motion=Free
  twistLimitDegrees=45
  twistMotion=Free
 }
 {
  bone1Name=dyn_shirt_bk_bott_l_MSTR
  bone2Name=spine_02
  swing1LimitDegrees=45
  swing1Motion=Free
  swing2LimitDegrees=45
  swing2Motion=Free
  twistLimitDegrees=45
  twistMotion=Free
 }
 {
  bone1Name=dyn_shirt_bk_bott_MSTR
  bone2Name=spine_02
  swing1LimitDegrees=45
  swing1Motion=Free
  swing2LimitDegrees=45
  swing2Motion=Free
  twistLimitDegrees=45
  twistMotion=Free
 }
 {
  bone1Name=dyn_shirt_bk_bott_r_MSTR
  bone2Name=spine_02
  swing1LimitDegrees=45
  swing1Motion=Free
  swing2LimitDegrees=45
  swing2Motion=Free
  twistLimitDegrees=45
  twistMotion=Free
 }
 {
  bone1Name=dyn_shirt_fr_l_MSTR
  bone2Name=spine_02
  swing1LimitDegrees=45
  swing1Motion=Free
  swing2LimitDegrees=45
  swing2Motion=Free
  twistLimitDegrees=45
  twistMotion=Free
 }
 {
  bone1Name=dyn_shirt_fr_MSTR
  bone2Name=spine_02
  swing1LimitDegrees=45
  swing1Motion=Free
  swing2LimitDegrees=45
  swing2Motion=Free
  twistLimitDegrees=45
  twistMotion=Free
 }
 {
  bone1Name=dyn_shirt_fr_r_MSTR
  bone2Name=spine_02
  swing1LimitDegrees=45
  swing1Motion=Free
  swing2LimitDegrees=45
  swing2Motion=Free
  twistLimitDegrees=45
  twistMotion=Free
 }
 {
  bone1Name=spine_02
  bone2Name=pelvis
  swing1LimitDegrees=45
  swing1Motion=Free
  swing2LimitDegrees=45
  swing2Motion=Free
  twistLimitDegrees=45
  twistMotion=Free
 }
 {
  bone1Name=thigh_l
  bone2Name=pelvis
  swing1LimitDegrees=45
  swing1Motion=Free
  swing2LimitDegrees=45
  swing2Motion=Free
  twistLimitDegrees=45
  twistMotion=Free
 }
 {
  bone1Name=thigh_r
  bone2Name=pelvis
  swing1LimitDegrees=45
  swing1Motion=Free
  swing2LimitDegrees=45
  swing2Motion=Free
  twistLimitDegrees=45
  twistMotion=Free
 }
 {
  bone1Name=spine_04
  bone2Name=spine_03
  swing1LimitDegrees=45
  swing1Motion=Free
  swing2LimitDegrees=45
  swing2Motion=Free
  twistLimitDegrees=45
  twistMotion=Free
 }
]
