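# Anim Blueprint for the Street Urchin body cosmetic, adding secondary motion to clothing, cans and zippers on the pose copied from the attached parent mesh.
# Weighted constraints drive the MSTR bones, then chained RigidBody sims in spine, thigh and foot space feed a cached pose.
# A backbling flag, set once on customization load, selects ShirtNoBack or full Shirt physics on that cache.
# Hoodie and collar constraint chains are left unconnected.
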AnimBlueprint /BRCosmetics/Characters/Player/Male/Medium/Bodies/M_MED_Street_Urchin/Meshes/M_MED_Street_Urchin_AnimBP
target_skeleton=&/M_MED_Street_Urchin_Skeleton
graphs [
 AnimGraph {nodes [
   n1 = "Constraint - dyn_hoodie_flap_r_MSTR (PPPPPPP)"!(
    ComponentPose
    Alpha=1
    ConstraintWeights_0=0.0268
    ConstraintWeights_1:float=0.4465
    ConstraintWeights_2:float=0.0635
    ConstraintWeights_3:float=0.2357
    ConstraintWeights_4:float=0.1146
    ConstraintWeights_5:float=0.0801
    ConstraintWeights_6:float=0.0327
   ) -> (Pose) {
    alphaBoolBlend {blendOption=Linear}
    boneToModify {boneName=dyn_hoodie_flap_r_MSTR}
    constraintSetup [
     {targetBone {boneName=head}}
     {targetBone {boneName=dyn_hoodie_flap_r}}
     {targetBone {boneName=spine_04}}
     {targetBone {boneName=spine_05}}
     {targetBone {boneName=clavicle_r}}
     {targetBone {boneName=neck_02}}
     {targetBone {boneName=neck_01}}
    ]
    constraintWeights [1,1,1,1,1,1,1]
   }
   "Constraint - dyn_hoodie_flap_l_MSTR (PPPPPPP)"!(
    ComponentPose=@n1
    Alpha=1
    ConstraintWeights_0=0.3934
    ConstraintWeights_1:float=0.0319
    ConstraintWeights_2:float=0.257
    ConstraintWeights_3:float=0.012
    ConstraintWeights_4:float=0.0942
    ConstraintWeights_5:float=0.1637
    ConstraintWeights_6:float=0.0478
   ) -> (Pose) {
    alphaBoolBlend {blendOption=Linear}
    boneToModify {boneName=dyn_hoodie_flap_l_MSTR}
    constraintSetup [
     {targetBone {boneName=head}}
     {targetBone {boneName=clavicle_l}}
     {targetBone {boneName=dyn_hoodie_flap_l}}
     {targetBone {boneName=spine_04}}
     {targetBone {boneName=spine_05}}
     {targetBone {boneName=neck_02}}
     {targetBone {boneName=neck_01}}
    ]
    constraintWeights [1,1,1,1,1,1,1]
   }
   n2 = "Constraint - dyn_collar_l_MSTR (PPPPPPPP)"!(
    ComponentPose
    Alpha=1
    ConstraintWeights_0=0
    ConstraintWeights_1:float=0.1973
    ConstraintWeights_2:float=0.0821
    ConstraintWeights_3:float=0.4687
    ConstraintWeights_4:float=0.2323
    ConstraintWeights_5:float=0.0107
    ConstraintWeights_6:float=0.0076
    ConstraintWeights_7:float=0.0013
   ) -> (Pose) {
    alphaBoolBlend {blendOption=Linear}
    boneToModify {boneName=dyn_collar_l_MSTR}
    constraintSetup [
     {targetBone {boneName=head}}
     {targetBone {boneName=clavicle_l}}
     {targetBone {boneName=spine_03}}
     {targetBone {boneName=spine_04}}
     {targetBone {boneName=spine_05}}
     {targetBone {boneName=clavicle_r}}
     {targetBone {boneName=neck_02}}
     {targetBone {boneName=neck_01}}
    ]
    constraintWeights [1,1,1,1,1,1,1,1]
   }
   "Constraint - dyn_collar_r_MSTR (PPPPPPPP)"!(
    ComponentPose=@n2
    Alpha=1
    ConstraintWeights_0=0
    ConstraintWeights_1:float=0.012
    ConstraintWeights_2:float=0.0523
    ConstraintWeights_3:float=0.4728
    ConstraintWeights_4:float=0.2536
    ConstraintWeights_5:float=0.1888
    ConstraintWeights_6:float=0.016
    ConstraintWeights_7:float=0.0044
   ) -> (Pose) {
    alphaBoolBlend {blendOption=Linear}
    boneToModify {boneName=dyn_collar_r_MSTR}
    constraintSetup [
     {targetBone {boneName=head}}
     {targetBone {boneName=clavicle_l}}
     {targetBone {boneName=spine_03}}
     {targetBone {boneName=spine_04}}
     {targetBone {boneName=spine_05}}
     {targetBone {boneName=clavicle_r}}
     {targetBone {boneName=neck_02}}
     {targetBone {boneName=neck_01}}
    ]
    constraintWeights [1,1,1,1,1,1,1,1]
   }
   n3 = RigidBody(ComponentPose=Local To Component(LocalPose=Use cached pose 'preJacketCache'())) {
    bClampLinearTranslationLimitToRefPose=true
    bOverrideWorldGravity=true
    overridePhysicsAsset=&/M_MED_Street_Urchin_Shirt
    simulationSpace=BaseBoneSpace
    alphaBoolBlend {blendOption=Linear}
    baseBoneRef {boneName=spine_01}
    componentAppliedLinearAccClamp=<3000 3000 3000>
    componentLinearAccScale=<0.3 0.25 0.25>
    componentLinearVelScale=<0.3 0.25 0.25>
    overrideWorldGravity {z=-980}
   }
   n4 = RigidBody(ComponentPose=Local To Component(LocalPose=Use cached pose 'preJacketCache'())) {
    bClampLinearTranslationLimitToRefPose=true
    bOverrideWorldGravity=true
    overridePhysicsAsset=&/M_MED_Street_Urchin_ShirtNoBack
    simulationSpace=BaseBoneSpace
    alphaBoolBlend {blendOption=Linear}
    baseBoneRef {boneName=spine_01}
    componentAppliedLinearAccClamp=<3000 3000 3000>
    componentLinearAccScale=<0.3 0.25 0.25>
    componentLinearVelScale=<0.3 0.25 0.25>
    overrideWorldGravity {z=-980}
   }
   Output Pose(Result=Blend Poses by bool(BlendPose_0=Component To Local(ComponentPose=@n4), BlendPose_1=Component To Local(ComponentPose=@n3), bActiveValue=Get isWearingBackbling()) {blendType=Linear}) {layerGroup=None}
   n5 = "Constraint - dyn_shirt_bk_top_l_MSTR (PPPPPP)"!(
    ComponentPose=Local To Component(LocalPose=Copy Pose From Mesh() {bUseAttachedParent=true})
    Alpha=1
    ConstraintWeights_0=0.2691
    ConstraintWeights_1:float=0.027
    ConstraintWeights_2:float=0.2837
    ConstraintWeights_3:float=0.4081
    ConstraintWeights_4:float=0.008
    ConstraintWeights_5:float=0.004
   ) -> (Pose) {
    alphaBoolBlend {blendOption=Linear}
    boneToModify {boneName=dyn_shirt_bk_top_l_MSTR}
    constraintSetup [
     {targetBone {boneName=clavicle_l}}
     {targetBone {boneName=spine_02}}
     {targetBone {boneName=spine_03}}
     {targetBone {boneName=spine_04}}
     {targetBone {boneName=spine_05}}
     {targetBone {boneName=clavicle_r}}
    ]
    constraintWeights [1,1,1,1,1,1]
   }
   n6 = "Constraint - dyn_shirt_bk_top_MSTR (PPPPPPPPPP)"!(
    ComponentPose=@n5
    Alpha=1
    ConstraintWeights_0=0.0149
    ConstraintWeights_1:float=0.0362
    ConstraintWeights_2:float=0.0529
    ConstraintWeights_3:float=0.0298
    ConstraintWeights_4:float=0.2911
    ConstraintWeights_5:float=0.4208
    ConstraintWeights_6:float=0.0758
    ConstraintWeights_7:float=0.0571
    ConstraintWeights_8:float=0.0068
    ConstraintWeights_9:float=0.0146
   ) -> (Pose) {
    alphaBoolBlend {blendOption=Linear}
    boneToModify {boneName=dyn_shirt_bk_top_MSTR}
    constraintSetup [
     {targetBone {boneName=dyn_shirt_bk_top_r_01}}
     {targetBone {boneName=dyn_shirt_bk_top_02}}
     {targetBone {boneName=clavicle_l}}
     {targetBone {boneName=spine_02}}
     {targetBone {boneName=spine_03}}
     {targetBone {boneName=spine_04}}
     {targetBone {boneName=spine_05}}
     {targetBone {boneName=clavicle_r}}
     {targetBone {boneName=dyn_shirt_bk_top_l_02}}
     {targetBone {boneName=dyn_shirt_bk_top_l_01}}
    ]
    constraintWeights [1,1,1,1,1,1,1,1,1,1]
   }
   n7 = "Constraint - dyn_shirt_bk_top_r_MSTR (PPPPP)"!(
    ComponentPose=@n6
    Alpha=1
    ConstraintWeights_0=0.1749
    ConstraintWeights_1:float=0.1205
    ConstraintWeights_2:float=0.0445
    ConstraintWeights_3:float=0.2557
    ConstraintWeights_4:float=0.4044
   ) -> (Pose) {
    alphaBoolBlend {blendOption=Linear}
    boneToModify {boneName=dyn_shirt_bk_top_r_MSTR}
    constraintSetup [
     {targetBone {boneName=clavicle_r}}
     {targetBone {boneName=dyn_shirt_bk_top_r_01}}
     {targetBone {boneName=spine_02}}
     {targetBone {boneName=spine_03}}
     {targetBone {boneName=spine_04}}
    ]
    constraintWeights [1,1,1,1,1]
   }
   n8 = "Constraint - dyn_shirt_bk_bott_l_MSTR (PPPPP)"!(
    ComponentPose=@n7
    Alpha=1
    ConstraintWeights_0=0.1485
    ConstraintWeights_1:float=0.3893
    ConstraintWeights_2:float=0.1642
    ConstraintWeights_3:float=0.0128
    ConstraintWeights_4:float=0.2852
   ) -> (Pose) {
    alphaBoolBlend {blendOption=Linear}
    boneToModify {boneName=dyn_shirt_bk_bott_l_MSTR}
    constraintSetup [
     {targetBone {boneName=pelvis}}
     {targetBone {boneName=spine_02}}
     {targetBone {boneName=spine_03}}
     {targetBone {boneName=spine_04}}
     {targetBone {boneName=spine_01}}
    ]
    constraintWeights [1,1,1,1,1]
   }
   n9 = "Constraint - dyn_shirt_bk_bott_MSTR (PPPPPP)"!(
    ComponentPose=@n8
    Alpha=1
    ConstraintWeights_0=0.5123
    ConstraintWeights_1:float=0.0784
    ConstraintWeights_2:float=0.2007
    ConstraintWeights_3:float=0.1543
    ConstraintWeights_4:float=0.0224
    ConstraintWeights_5:float=0.032
   ) -> (Pose) {
    alphaBoolBlend {blendOption=Linear}
    boneToModify {boneName=dyn_shirt_bk_bott_MSTR}
    constraintSetup [
     {targetBone {boneName=dyn_shirt_bk_bott_01}}
     {targetBone {boneName=spine_01}}
     {targetBone {boneName=spine_02}}
     {targetBone {boneName=spine_03}}
     {targetBone {boneName=spine_04}}
     {targetBone {boneName=pelvis}}
    ]
    constraintWeights [1,1,1,1,1,1]
   }
   n10 = "Constraint - dyn_shirt_bk_bott_r_MSTR (PPPPP)"!(
    ComponentPose=@n9
    Alpha=1
    ConstraintWeights_0=0.1085
    ConstraintWeights_1:float=0.4111
    ConstraintWeights_2:float=0.1824
    ConstraintWeights_3:float=0.0119
    ConstraintWeights_4:float=0.2861
   ) -> (Pose) {
    alphaBoolBlend {blendOption=Linear}
    boneToModify {boneName=dyn_shirt_bk_bott_r_MSTR}
    constraintSetup [
     {targetBone {boneName=pelvis}}
     {targetBone {boneName=spine_02}}
     {targetBone {boneName=spine_03}}
     {targetBone {boneName=spine_04}}
     {targetBone {boneName=spine_01}}
    ]
    constraintWeights [1,1,1,1,1]
   }
   n11 = "Constraint - dyn_shirt_fr_l_MSTR (PPPPPPP)"!(
    ComponentPose=@n10
    Alpha=1
    ConstraintWeights_0=0.1581
    ConstraintWeights_1:float=0.0935
    ConstraintWeights_2:float=0.3061
    ConstraintWeights_3:float=0.1562
    ConstraintWeights_4:float=0.0538
    ConstraintWeights_5:float=0.0595
    ConstraintWeights_6:float=0.1729
   ) -> (Pose) {
    alphaBoolBlend {blendOption=Linear}
    boneToModify {boneName=dyn_shirt_fr_l_MSTR}
    constraintSetup [
     {targetBone {boneName=dyn_shirt_fr_l_01}}
     {targetBone {boneName=dyn_shirt_fr_l_02}}
     {targetBone {boneName=spine_01}}
     {targetBone {boneName=spine_02}}
     {targetBone {boneName=spine_03}}
     {targetBone {boneName=dyn_shirt_fr_01}}
     {targetBone {boneName=pelvis}}
    ]
    constraintWeights [1,1,1,1,1,1,1]
   }
   n12 = "Constraint - dyn_shirt_fr_MSTR (PPPPPPPPPP)"!(
    ComponentPose=@n11
    Alpha=1
    ConstraintWeights_0=0.0122
    ConstraintWeights_1:float=0.0169
    ConstraintWeights_2:float=0.0256
    ConstraintWeights_3:float=0.0119
    ConstraintWeights_4:float=0.2883
    ConstraintWeights_5:float=0.1295
    ConstraintWeights_6:float=0.0386
    ConstraintWeights_7:float=0.2456
    ConstraintWeights_8:float=0.0572
    ConstraintWeights_9:float=0.1741
   ) -> (Pose) {
    alphaBoolBlend {blendOption=Linear}
    boneToModify {boneName=dyn_shirt_fr_MSTR}
    constraintSetup [
     {targetBone {boneName=dyn_shirt_fr_r_02}}
     {targetBone {boneName=dyn_shirt_fr_l_01}}
     {targetBone {boneName=dyn_shirt_fr_l_02}}
     {targetBone {boneName=dyn_shirt_fr_r_01}}
     {targetBone {boneName=spine_01}}
     {targetBone {boneName=spine_02}}
     {targetBone {boneName=spine_03}}
     {targetBone {boneName=dyn_shirt_fr_01}}
     {targetBone {boneName=dyn_shirt_fr_02}}
     {targetBone {boneName=pelvis}}
    ]
    constraintWeights [1,1,1,1,1,1,1,1,1,1]
   }
   n13 = "Constraint - dyn_shirt_fr_r_MSTR (PPPPPPPP)"!(
    ComponentPose=@n12
    Alpha=1
    ConstraintWeights_0=0.0907
    ConstraintWeights_1:float=0.1986
    ConstraintWeights_2:float=0.2861
    ConstraintWeights_3:float=0.1293
    ConstraintWeights_4:float=0.0419
    ConstraintWeights_5:float=0.0656
    ConstraintWeights_6:float=0.0217
    ConstraintWeights_7:float=0.1661
   ) -> (Pose) {
    alphaBoolBlend {blendOption=Linear}
    boneToModify {boneName=dyn_shirt_fr_r_MSTR}
    constraintSetup [
     {targetBone {boneName=dyn_shirt_fr_r_02}}
     {targetBone {boneName=dyn_shirt_fr_r_01}}
     {targetBone {boneName=spine_01}}
     {targetBone {boneName=spine_02}}
     {targetBone {boneName=spine_03}}
     {targetBone {boneName=dyn_shirt_fr_01}}
     {targetBone {boneName=dyn_shirt_fr_02}}
     {targetBone {boneName=pelvis}}
    ]
    constraintWeights [1,1,1,1,1,1,1,1]
   }
   n14 = "Constraint - dyn_can_chest_01_MSTR (PPPPP)"!(
    ComponentPose=@n13
    Alpha=1
    ConstraintWeights_0=0.1144
    ConstraintWeights_1:float=0.1968
    ConstraintWeights_2:float=0.571
    ConstraintWeights_3:float=0.0373
    ConstraintWeights_4:float=0.0805
   ) -> (Pose) {
    alphaBoolBlend {blendOption=Linear}
    boneToModify {boneName=dyn_can_chest_01_MSTR}
    constraintSetup [
     {targetBone {boneName=spine_02}}
     {targetBone {boneName=spine_03}}
     {targetBone {boneName=spine_04}}
     {targetBone {boneName=spine_05}}
     {targetBone {boneName=clavicle_l}}
    ]
    constraintWeights [1,1,1,1,1]
   }
   n15 = "Constraint - dyn_can_chest_02_MSTR (PPPPP)"!(
    ComponentPose=@n14
    Alpha=1
    ConstraintWeights_0=0.0881
    ConstraintWeights_1:float=0.1509
    ConstraintWeights_2:float=0.4944
    ConstraintWeights_3:float=0.0629
    ConstraintWeights_4:float=0.2037
   ) -> (Pose) {
    alphaBoolBlend {blendOption=Linear}
    boneToModify {boneName=dyn_can_chest_02_MSTR}
    constraintSetup [
     {targetBone {boneName=spine_02}}
     {targetBone {boneName=spine_03}}
     {targetBone {boneName=spine_04}}
     {targetBone {boneName=spine_05}}
     {targetBone {boneName=clavicle_l}}
    ]
    constraintWeights [1,1,1,1,1]
   }
   n16 = "Constraint - dyn_can_chest_03_MSTR (PPPP)"!(
    ComponentPose=@n15
    Alpha=1
    ConstraintWeights_0=0.1648
    ConstraintWeights_1:float=0.3424
    ConstraintWeights_2:float=0.1159
    ConstraintWeights_3:float=0.3768
   ) -> (Pose) {
    alphaBoolBlend {blendOption=Linear}
    boneToModify {boneName=dyn_can_chest_03_MSTR}
    constraintSetup [{targetBone {boneName=spine_03}}, {targetBone {boneName=spine_04}}, {targetBone {boneName=spine_05}}, {targetBone {boneName=clavicle_l}}]
    constraintWeights [1,1,1,1]
   }
   n17 = "Constraint - dyn_can_hip_01_MSTR (PPP)"!(ComponentPose=@n16, Alpha=1, ConstraintWeights_0=0.0098, ConstraintWeights_1:float=0.4686, ConstraintWeights_2:float=0.5216) -> (Pose) {
    alphaBoolBlend {blendOption=Linear}
    boneToModify {boneName=dyn_can_hip_01_MSTR}
    constraintSetup [{targetBone {boneName=spine_01}}, {targetBone {boneName=thigh_twist_01_l}}, {targetBone {boneName=pelvis}}]
    constraintWeights [1,1,1]
   }
   n18 = "Constraint - dyn_can_hip_02_MSTR (PPP)"!(ComponentPose=@n17, Alpha=1, ConstraintWeights_0=0.0012, ConstraintWeights_1:float=0.3456, ConstraintWeights_2:float=0.6532) -> (Pose) {
    alphaBoolBlend {blendOption=Linear}
    boneToModify {boneName=dyn_can_hip_02_MSTR}
    constraintSetup [{targetBone {boneName=spine_01}}, {targetBone {boneName=thigh_twist_01_l}}, {targetBone {boneName=pelvis}}]
    constraintWeights [1,1,1]
   }
   n19 = "Constraint - dyn_zipper_chest_MSTR (PPPPPP)"!(
    ComponentPose=@n18
    Alpha=1
    ConstraintWeights_0=0.03
    ConstraintWeights_1:float=0.0944
    ConstraintWeights_2:float=0.1733
    ConstraintWeights_3:float=0.633
    ConstraintWeights_4:float=0.0394
    ConstraintWeights_5:float=0.0299
   ) -> (Pose) {
    alphaBoolBlend {blendOption=Linear}
    boneToModify {boneName=dyn_zipper_chest_MSTR}
    constraintSetup [
     {targetBone {boneName=clavicle_l}}
     {targetBone {boneName=spine_02}}
     {targetBone {boneName=spine_03}}
     {targetBone {boneName=spine_04}}
     {targetBone {boneName=spine_05}}
     {targetBone {boneName=clavicle_r}}
    ]
    constraintWeights [1,1,1,1,1,1]
   }
   n20 = "Constraint - dyn_zipper_front_groin_MSTR (PPPP)"!(
    ComponentPose=@n19
    Alpha=1
    ConstraintWeights_0=0.0626
    ConstraintWeights_1:float=0.1127
    ConstraintWeights_2:float=0.0849
    ConstraintWeights_3:float=0.7398
   ) -> (Pose) {
    alphaBoolBlend {blendOption=Linear}
    boneToModify {boneName=dyn_zipper_front_groin_MSTR}
    constraintSetup [
     {targetBone {boneName=spine_01}}
     {targetBone {boneName=thigh_twist_01_l}}
     {targetBone {boneName=thigh_twist_01_r}}
     {targetBone {boneName=pelvis}}
    ]
    constraintWeights [1,1,1,1]
   }
   n21 = "Constraint - dyn_zipper_thigh_l_MSTR (PPP)"!(ComponentPose=@n20, Alpha=1, ConstraintWeights_0=0.0244, ConstraintWeights_1:float=0.8642, ConstraintWeights_2:float=0.1114) -> (Pose) {
    alphaBoolBlend {blendOption=Linear}
    boneToModify {boneName=dyn_zipper_thigh_l_MSTR}
    constraintSetup [{targetBone {boneName=pelvis}}, {targetBone {boneName=thigh_twist_01_l}}, {targetBone {boneName=thigh_l}}]
    constraintWeights [1,1,1]
   }
   n22 = "Constraint - dyn_zipper_thigh_r_MSTR (PPP)"!(ComponentPose=@n21, Alpha=1, ConstraintWeights_0=0.1082, ConstraintWeights_1:float=0.0279, ConstraintWeights_2:float=0.864) -> (Pose) {
    alphaBoolBlend {blendOption=Linear}
    boneToModify {boneName=dyn_zipper_thigh_r_MSTR}
    constraintSetup [{targetBone {boneName=thigh_r}}, {targetBone {boneName=pelvis}}, {targetBone {boneName=thigh_twist_01_r}}]
    constraintWeights [1,1,1]
   }
   n23 = "Constraint - dyn_zipper_butt_MSTR (PPP)"!(ComponentPose=@n22, Alpha=1, ConstraintWeights_0=0.8389, ConstraintWeights_1:float=0.0166, ConstraintWeights_2:float=0.1445) -> (Pose) {
    alphaBoolBlend {blendOption=Linear}
    boneToModify {boneName=dyn_zipper_butt_MSTR}
    constraintSetup [{targetBone {boneName=pelvis}}, {targetBone {boneName=thigh_twist_01_l}}, {targetBone {boneName=thigh_twist_01_r}}]
    constraintWeights [1,1,1]
   }
   n24 = "Constraint - dyn_pants_flap_r_bott_MSTR (PPPP)"!(
    ComponentPose=@n23
    Alpha=1
    ConstraintWeights_0=0.0252
    ConstraintWeights_1:float=0.0744
    ConstraintWeights_2:float=0.237
    ConstraintWeights_3:float=0.6635
   ) -> (Pose) {
    alphaBoolBlend {blendOption=Linear}
    boneToModify {boneName=dyn_pants_flap_r_bott_MSTR}
    constraintSetup [
     {targetBone {boneName=thigh_r}}
     {targetBone {boneName=pelvis}}
     {targetBone {boneName=dyn_pants_flap_r_bott}}
     {targetBone {boneName=thigh_twist_01_r}}
    ]
    constraintWeights [1,1,1,1]
   }
   n25 = "Constraint - dyn_pants_flap_l_bott_MSTR (PPPP)"!(
    ComponentPose=@n24
    Alpha=1
    ConstraintWeights_0=0.0744
    ConstraintWeights_1:float=0.6635
    ConstraintWeights_2:float=0.237
    ConstraintWeights_3:float=0.0252
   ) -> (Pose) {
    alphaBoolBlend {blendOption=Linear}
    boneToModify {boneName=dyn_pants_flap_l_bott_MSTR}
    constraintSetup [
     {targetBone {boneName=pelvis}}
     {targetBone {boneName=thigh_twist_01_l}}
     {targetBone {boneName=dyn_pants_flap_l_bott}}
     {targetBone {boneName=thigh_l}}
    ]
    constraintWeights [1,1,1,1]
   }
   n26 = "Constraint - dyn_pants_flap_r_top_MSTR (PP)"!(ComponentPose=@n25, Alpha=1, ConstraintWeights_0=0.2177, ConstraintWeights_1:float=0.7823) -> (Pose) {
    alphaBoolBlend {blendOption=Linear}
    boneToModify {boneName=dyn_pants_flap_r_top_MSTR}
    constraintSetup [{targetBone {boneName=pelvis}}, {targetBone {boneName=thigh_twist_01_r}}]
    constraintWeights [1,1]
   }
   n27 = "Constraint - dyn_pants_flap_l_top_MSTR (PP)"!(ComponentPose=@n26, Alpha=1, ConstraintWeights_0=0.2177, ConstraintWeights_1:float=0.7823) -> (Pose) {
    alphaBoolBlend {blendOption=Linear}
    boneToModify {boneName=dyn_pants_flap_l_top_MSTR}
    constraintSetup [{targetBone {boneName=pelvis}}, {targetBone {boneName=thigh_twist_01_l}}]
    constraintWeights [1,1]
   }
   n28 = RigidBody(ComponentPose=@n27) {
    bClampLinearTranslationLimitToRefPose=true
    bOverrideWorldGravity=true
    overridePhysicsAsset=&/M_MED_Street_Urchin_Dyn
    simulationSpace=BaseBoneSpace
    alphaBoolBlend {blendOption=Linear}
    baseBoneRef {boneName=spine_01}
    componentAppliedLinearAccClamp=<3000 3000 3000>
    componentLinearAccScale=<0.6 0.5 0.5>
    componentLinearVelScale=<0.6 0.5 0.5>
    overrideWorldGravity {z=-980}
   }
   n29 = RigidBody(ComponentPose=@n28) {
    bClampLinearTranslationLimitToRefPose=true
    bOverrideWorldGravity=true
    overridePhysicsAsset=&/M_MED_Street_Urchin_Dyn_leg_l
    simulationSpace=BaseBoneSpace
    alphaBoolBlend {blendOption=Linear}
    baseBoneRef {boneName=thigh_l}
    componentAppliedLinearAccClamp=<3000 3000 3000>
    componentLinearAccScale=<0.55 0.3 0.3>
    componentLinearVelScale=<0.55 0.3 0.3>
    overrideWorldGravity {z=-980}
   }
   n30 = RigidBody(ComponentPose=@n29) {
    bClampLinearTranslationLimitToRefPose=true
    bOverrideWorldGravity=true
    overridePhysicsAsset=&/M_MED_Street_Urchin_Dyn_leg_r
    simulationSpace=BaseBoneSpace
    alphaBoolBlend {blendOption=Linear}
    baseBoneRef {boneName=thigh_r}
    componentAppliedLinearAccClamp=<3000 3000 3000>
    componentLinearAccScale=<0.55 0.3 0.3>
    componentLinearVelScale=<0.55 0.3 0.3>
    overrideWorldGravity {z=-980}
   }
   n31 = RigidBody(ComponentPose=@n30) {
    bClampLinearTranslationLimitToRefPose=true
    bOverrideWorldGravity=true
    overridePhysicsAsset=&/M_MED_Street_Urchin_Dyn_foot_l
    simulationSpace=BaseBoneSpace
    alphaBoolBlend {blendOption=Linear}
    baseBoneRef {boneName=foot_l}
    componentAppliedLinearAccClamp=<3000 3000 3000>
    componentLinearAccScale=<0.6 0.5 0.5>
    componentLinearVelScale=<0.6 0.5 0.5>
    overrideWorldGravity {z=-980}
   }
   n32 = RigidBody(ComponentPose=@n31) {
    bClampLinearTranslationLimitToRefPose=true
    bOverrideWorldGravity=true
    overridePhysicsAsset=&/M_MED_Street_Urchin_Dyn_foot_r
    simulationSpace=BaseBoneSpace
    alphaBoolBlend {blendOption=Linear}
    baseBoneRef {boneName=foot_r}
    componentAppliedLinearAccClamp=<3000 3000 3000>
    componentLinearAccScale=<0.6 0.5 0.5>
    componentLinearVelScale=<0.6 0.5 0.5>
    overrideWorldGravity {z=-980}
   }
   Save cached pose 'preJacketCache'(Pose=Component To Local(ComponentPose=@n32)) {cacheName=preJacketCache}
  ]}
 EventGraph {nodes [
   Event OnCharacterCustomizationLoaded()
   > Do Once()
   n1 = CheckForBackbling()
   > Set isWearingBackbling(isWearingBackbling=OR Boolean(A=@n1, B=@n1.bIsBackblingLong))
   TryGetPawnOwner()
   n2 = GetPartSkeletalMeshComponent()
   "Not Equal (Object)"!(
    A:Object Reference=FindChildCommponentFromAttachSocket!(
     self:Custom Character Part Anim Instance Object Reference
     Parent:Skeletal Mesh Component Object Reference=@n2
     bIncludeAllDescendants:bool=false
     AttachSocket:Name=Cape
    ) -> (ReturnValue:Skeletal Mesh Component Object Reference)
    B:Object Reference
   ) -> (ReturnValue:bool)
   "Not Equal (Object)"(A=FindChildCommponentFromAttachSocket(Parent=@n2, bIncludeAllDescendants=false, AttachSocket=backpack_br))
   Event BlueprintUpdateAnimation()
  ]}
]
variables [isWearingBackbling {type=boolean}]
